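# Niagara system for the Nucleus pickaxe idle effect, adding start and looping sheen sweeps with glints.
# SheenMesh emitters burst one mesh with animated material parameters, plus two sprite renderers and curve-driven flare variables.
# They generate death events; Glint emitters have spawn rate disabled and show no event handlers.
# Glint sprites take location, size and color from shared System variables, then spin down and shrink.
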NiagaraSystem /BRCosmetics/Weapons/FORT_Melee/Pickaxe_Nucleus_Male/FX/NS_Pickaxe_Nucleus_Idle
dependencies {
 usedDataInterfaces [&NiagaraDataInterfaceVectorCurve, &NiagaraDataInterfaceCurve, &NiagaraDataInterfaceColorCurve]
 usedDynamicInputs [&Vector2DFromFloat, &Multiply_Vector2]
 usedModules [
  &self:SystemSpawnScript.SystemScriptSource.SystemScriptGraph.NiagaraNodeAssignment_0.SetVariables_#1
  &SystemState
 ]
}
emitters {
 Glint_Loop {
  emitterUpdateScript [
   &EmitterState(Life Cycle Mode=::NewEnumerator1, Loop Duration=1, UseLoopDelay=true)
   &SpawnRate(SpawnRate=200, Use Spawn Probability=false) {enabled=false}
  ]
  particleSpawnScript [
   /SetVariables(Particles.SpriteRotationRate=&RandomRangeFloat(Minimum=140, Maximum=-140, Evaluation Type=::NewEnumerator1)) {isSetParameters=true, module=SetVariables_#2}
   /Niagara/Modules/Spawn/Initialization/InitializeParticle(
    Lifetime=&RandomRangeFloat(Minimum=0.3, Maximum=0.7, Evaluation Type=::NewEnumerator1)
    Module.Write Color=true
    Module.Write Lifetime=true
    Module.Write Mass=true
    Module.Write Material Random=true
    Module.Write Position=true
    Module.Write Scale=false
    Module.Write SpriteRotation=true
    Module.Write SpriteSize=true
    Position:Vector3f=$System.GlintLocation
    Sprite Rotation=&RandomRangeFloat(Minimum=0, Maximum=0.5, Evaluation Type=::NewEnumerator1)
    Sprite Size=&RandomRangeVector2D(Minimum=<56.25 33.75>, Maximum=<78.75 56.25>, Evaluation Type=::NewEnumerator1)
   )
   /SetVariables(Particles.SpriteSize=$System.SystemGlintSize) {isSetParameters=true, module=SetVariables_#3}
   &Color(Color=$System.GlintColor)
  ]
  particleUpdateScript [
   &ParticleState()
   &ScaleSpriteSize(Uniform Curve Sprite Scale=DataInterface() {propertyValues {Curve {keys [
        <at 0 0 RCIM_Cubic>
        {arriveTangent=1.6862, interpMode=RCIM_Cubic, leaveTangent=1.6862, time=0.0545, value=0.9401}
        {arriveTangent=-0.9943, interpMode=RCIM_Cubic, leaveTangent=-0.9943, time=0.3437, value=0.5796}
        <at 1 0>
       ]}}})
   &CameraOffset(Camera Offset Amount=&FloatFromCurve(FloatCurve=DataInterface() {propertyValues {Curve {keys [<at 0 15>,<at 1 20>]}}}, Scale Curve=2))
   &SpriteRotationRate(Rotation Rate=&FloatFromCurve(FloatCurve=DataInterface() {propertyValues {Curve {keys [<at 0 1>,<at 1 0>]}}}, Scale Curve=$Particles.SpriteRotationRate))
  ]
  rendererClasses [&NiagaraSpriteRendererProperties]
  renderers [{rendererClass=&NiagaraSpriteRendererProperties}]
 }
 Glint_Start {
  emitterUpdateScript [
   &EmitterState(Life Cycle Mode=::NewEnumerator1, Loop Duration=1, UseLoopDelay=true)
   &SpawnRate(SpawnRate=200, Use Spawn Probability=false) {enabled=false}
  ]
  particleSpawnScript [
   /SetVariables(Particles.SpriteRotationRate=&RandomRangeFloat(Minimum=360, Maximum=500, Evaluation Type=::NewEnumerator1)) {isSetParameters=true, module=SetVariables_#2}
   &InitializeParticle(
    Lifetime=&RandomRangeFloat(Minimum=0.3, Maximum=0.7, Evaluation Type=::NewEnumerator1)
    Module.Write Color=true
    Module.Write Lifetime=true
    Module.Write Mass=true
    Module.Write Material Random=true
    Module.Write Position=true
    Module.Write Scale=false
    Module.Write SpriteRotation=true
    Module.Write SpriteSize=true
    Position:Vector3f=$System.GlintLocation
    Sprite Rotation=&RandomRangeFloat(Minimum=0, Maximum=0.5, Evaluation Type=::NewEnumerator1)
    Sprite Size=&RandomRangeVector2D(Minimum=<56.25 33.75>, Maximum=<78.75 56.25>, Evaluation Type=::NewEnumerator1)
   )
   /SetVariables(Particles.SpriteSize=$System.SystemGlintSize) {isSetParameters=true, module=SetVariables_#3}
   &Color(Color=$System.GlintColor)
  ]
  particleUpdateScript [
   &ParticleState()
   &ScaleSpriteSize(Uniform Curve Sprite Scale=DataInterface() {propertyValues {Curve {keys [
        <at 0 0 RCIM_Cubic>
        {arriveTangent=1.6862, interpMode=RCIM_Cubic, leaveTangent=1.6862, time=0.0545, value=0.9401}
        {arriveTangent=-0.9943, interpMode=RCIM_Cubic, leaveTangent=-0.9943, time=0.3437, value=0.5796}
        <at 1 0>
       ]}}})
   &CameraOffset(Camera Offset Amount=&FloatFromCurve(FloatCurve=DataInterface() {propertyValues {Curve {keys [<at 0 15>,<at 1 20>]}}}, Scale Curve=2))
   &SpriteRotationRate(Rotation Rate=&FloatFromCurve(FloatCurve=DataInterface() {propertyValues {Curve {keys [<at 0 1>,<at 1 0>]}}}, Scale Curve=$Particles.SpriteRotationRate))
  ]
  rendererClasses [&NiagaraSpriteRendererProperties]
  renderers [{rendererClass=&NiagaraSpriteRendererProperties}]
 }
 SheenMesh_loop {
  emitterUpdateScript [
   &EmitterState(
    Delay First Loop Only=true
    Life Cycle Mode=::NewEnumerator1
    Loop Delay=&RandomRangeFloat(Minimum=3, Maximum=2, Evaluation Type=::NewEnumerator1)
    Loop Duration=&RandomRangeFloat(Minimum=5, Maximum=8, Evaluation Type=::NewEnumerator1)
    Recalculate Duration Each Loop=true
    UseLoopDelay=true
   )
   &SpawnBurst_Instantaneous(Use Spawn Probability=false, UseLoopCountLimit=false)
   &SpawnRate(SpawnRate=20, Use Spawn Probability=false) {enabled=false}
  ]
  particleSpawnScript [
   &Color(Color=&MakeLinearColorFromVectorAndFloat(
     "Vector (RGB)"=<1 0 0>
     "Float (Alpha)"=&ClampFloat(Float=&RandomRangeFloat(Minimum=0.4, Maximum=0.6, Evaluation Type=::NewEnumerator1))
    )) {enabled=false}
   /SetVariables(
    Particles.ColorFlare:LinearColor=&MultiplyLinearColor(Linear Color=$System.GlintColor, Scale Factor=<20 20 20 20>)
    Particles.PositionFlare:NiagaraPosition=$Particles.Position
   ) {isSetParameters=true, module=SetVariables_#4}
   &InitializeParticle(
    Color=<rgba 0.8 1 0.8333 1>
    Lifetime=&RandomRangeFloat(Minimum=0.25, Maximum=0.4)
    Module.Write Color=true
    Module.Write Lifetime=true
    Module.Write Mass=true
    Module.Write Material Random=true
    Module.Write Position=true
    Module.Write Scale=true
    Module.Write SpriteRotation=false
    Module.Write SpriteSize=true
    Position=<-0.5 0 2>
    Sprite Size=<1 2>
   )
   /NiagaraFN/Sprite/FN_SpriteAlignmentAndFacing(SpriteAlignment:Vector3f=<0 0 1>, SpriteFacing:Vector3f=<1 0 0>)
  ]
  particleUpdateScript [
   /SetVariables(
    Particles.PositionFlare:Vector3f=&VectorFromCurve(VectorCurve=DataInterface() {propertyValues {
       XCurve {keys [
         <at 0.0009 0 RCIM_Cubic>
         {arriveTangent=15.0216, interpMode=RCIM_Cubic, leaveTangent=15.0216, time=0.4, value=2.5}
         {arriveTangent=11.1923, interpMode=RCIM_Cubic, leaveTangent=11.1929, tangentMode=RCTM_User, time=0.6, value=9}
        ]}
       YCurve {keys [<at 0 0>,<at 1 0>]}
       ZCurve {keys [
         <at 0 35 RCIM_Cubic>
         {arriveTangent=94.8541, interpMode=RCIM_Cubic, leaveTangent=94.8537, tangentMode=RCTM_User, time=0.35, value=55}
         {arriveTangent=34.8214, interpMode=RCIM_Cubic, leaveTangent=34.8228, tangentMode=RCTM_User, time=0.6, value=75}
        ]}
      }})
    Particles.PositionFlare2:Vector3f=&MakeVector(
     X=1
     Y=&FloatFromCurve(FloatCurve=DataInterface() {propertyValues {Curve {keys [
          {arriveTangent=134.5043, interpMode=RCIM_Cubic, leaveTangent=134.5037, tangentMode=RCTM_User, time=0, value=15}
          {arriveTangent=33.5313, interpMode=RCIM_Cubic, leaveTangent=33.5312, tangentMode=RCTM_User, time=0.5, value=30}
         ]}}})
     Z=69
    )
   ) {isSetParameters=true, module=SetVariables_#5}
   &ParticleState()
   &ScaleSpriteSize(Uniform Curve Sprite Scale=DataInterface() {propertyValues {Curve {keys [
        {arriveTangent=7.953, interpMode=RCIM_Cubic, leaveTangent=7.953, tangentMode=RCTM_User, time=-0.0008, value=0.5}
        {interpMode=RCIM_Cubic, leaveTangent=-0.7143, time=0.225, value=1.8998}
        {arriveTangent=-0.1355, interpMode=RCIM_Cubic, leaveTangent=-0.1355, tangentMode=RCTM_User, time=0.6, value=0.1497}
       ]}}})
   /SetVariables(Particles.SpriteSizeFlare2:Vector2f=&MakeVector2D(
     X=&FloatFromCurve(
      FloatCurve=DataInterface() {propertyValues {Curve {keys [{arriveTangent=-2.2776, interpMode=RCIM_Cubic, leaveTangent=-2.2776, tangentMode=RCTM_User, time=0, value=1}, <at 1 0 RCIM_Cubic>]}}}
      Scale Curve=3
     )
     Y=&FloatFromCurve(
      FloatCurve=DataInterface() {propertyValues {Curve {keys [<at 0 1 RCIM_Cubic>, {arriveTangent=-1.3497, interpMode=RCIM_Cubic, leaveTangent=-1.3497, tangentMode=RCTM_User, time=1, value=0.5}]}}}
      Scale Curve=20
     )
    )) {isSetParameters=true, module=SetVariables_#6}
   &Color(Color=&MakeLinearColorFromVectorAndFloat("Vector (RGB)"=<0.4208 0.6361 1>, "Float (Alpha)"=&ClampFloat(Float=0.8))) {enabled=false, module=Color001}
   &DynamicMaterialParameters(
    Index 0 Param 1=&FloatFromCurve(FloatCurve=DataInterface() {propertyValues {Curve {keys [{arriveTangent=2.513, interpMode=RCIM_Cubic, time=0, value=0.2}, <at 1 1.8>]}}})
    Index 0 Param 2=1
    Index 0 Param 3=1
    Index 0 Param 4=1
    Module.Param0WriteEnabled=true
    Module.Param1WriteEnabled=true
    Module.Param2WriteEnabled=true
    Module.Param3WriteEnabled=true
   )
   &CameraOffset(Camera Offset Amount=1)
   /SetVariables(Particles.CameraOffsetFlare=2) {isSetParameters=true, module=SetVariables_#7}
   /SetVariables(Particles.ColorFlare:LinearColor=&ColorFromCurve(ColorCurve=DataInterface() {propertyValues {
       AlphaCurve {keys [<at 0 1>, {arriveTangent=-1, interpMode=RCIM_Cubic, leaveTangent=-1, time=0.5882, value=1}, <at 1 0 RCIM_Cubic>]}
       BlueCurve {keys [<at 0 0.7292 RCIM_Cubic>,<at 0.6045 0.599 RCIM_Cubic>]}
       GreenCurve {keys [<at 0 0.7292 RCIM_Cubic>,<at 0.6045 0.593 RCIM_Cubic>]}
       RedCurve {keys [<at 0 0.7292 RCIM_Cubic>,<at 0.6045 0.4678 RCIM_Cubic>]}
      }})) {isSetParameters=true, module=SetVariables_#8}
   &SolveForcesAndVelocity(Clamp Velocity=false, Limit Acceleration=false)
   &GenerateDeathEvent(Event Probability=1.5, Use Event Probability=true)
  ]
  rendererClasses [&NiagaraMeshRendererProperties, &NiagaraSpriteRendererProperties]
  renderers [
   {rendererClass=&NiagaraMeshRendererProperties, rendererIndex=0}
   {rendererClass=&NiagaraSpriteRendererProperties, rendererIndex=1}
   {rendererClass=&NiagaraSpriteRendererProperties, rendererIndex=2}
  ]
 }
 SheenMesh_start {
  emitterUpdateScript [
   &EmitterState(
    Life Cycle Mode=::NewEnumerator1
    Loop Behavior=::NewEnumerator1
    Loop Delay=&RandomRangeFloat(Minimum=0, Maximum=0.25, Evaluation Type=::NewEnumerator1)
    Loop Duration=0.5
    UseLoopDelay=true
   )
   &SpawnBurst_Instantaneous(Use Spawn Probability=false, UseLoopCountLimit=false)
   &SpawnRate(SpawnRate=20, Use Spawn Probability=false) {enabled=false}
  ]
  particleSpawnScript [
   &Color(Color=&MakeLinearColorFromVectorAndFloat(
     "Vector (RGB)"=<1 0 0>
     "Float (Alpha)"=&ClampFloat(Float=&RandomRangeFloat(Minimum=0.4, Maximum=0.6, Evaluation Type=::NewEnumerator1))
    )) {enabled=false}
   /SetVariables(
    Particles.ColorFlare:LinearColor=&MultiplyLinearColor(Linear Color=$System.GlintColor, Scale Factor=<20 20 20 20>)
    Particles.PositionFlare:NiagaraPosition=$Particles.Position
   ) {isSetParameters=true, module=SetVariables_#4}
   &InitializeParticle(
    Color=<rgba 0.8 1 0.8333 1>
    Lifetime=&RandomRangeFloat(Minimum=0.25, Maximum=0.4)
    Module.Write Color=true
    Module.Write Lifetime=true
    Module.Write Mass=true
    Module.Write Material Random=true
    Module.Write Position=true
    Module.Write Scale=true
    Module.Write SpriteRotation=false
    Module.Write SpriteSize=true
    Position=<-0.5 0 2>
    Sprite Size=<1 2>
   )
   &FN_SpriteAlignmentAndFacing(SpriteAlignment:Vector3f=<0 0 1>, SpriteFacing:Vector3f=<1 0 0>)
  ]
  particleUpdateScript [
   /SetVariables(
    Particles.PositionFlare:Vector3f=&VectorFromCurve(VectorCurve=DataInterface() {propertyValues {
       XCurve {keys [
         <at 0.0009 0 RCIM_Cubic>
         {arriveTangent=15.0216, interpMode=RCIM_Cubic, leaveTangent=15.0216, time=0.4, value=2.5}
         {arriveTangent=11.1923, interpMode=RCIM_Cubic, leaveTangent=11.1929, tangentMode=RCTM_User, time=0.6, value=9}
        ]}
       YCurve {keys [<at 0 0>,<at 1 0>]}
       ZCurve {keys [
         <at 0 35 RCIM_Cubic>
         {arriveTangent=94.8541, interpMode=RCIM_Cubic, leaveTangent=94.8537, tangentMode=RCTM_User, time=0.35, value=55}
         {arriveTangent=34.8214, interpMode=RCIM_Cubic, leaveTangent=34.8228, tangentMode=RCTM_User, time=0.6, value=75}
        ]}
      }})
    Particles.PositionFlare2:Vector3f=&MakeVector(
     X=1
     Y=&FloatFromCurve(FloatCurve=DataInterface() {propertyValues {Curve {keys [
          {arriveTangent=107.135, interpMode=RCIM_Cubic, leaveTangent=107.1348, tangentMode=RCTM_User, time=0, value=15}
          {arriveTangent=33.5313, interpMode=RCIM_Cubic, leaveTangent=33.5312, tangentMode=RCTM_User, time=0.5, value=30}
         ]}}})
     Z=69
    )
   ) {isSetParameters=true, module=SetVariables_#5}
   &ParticleState()
   &ScaleSpriteSize(Uniform Curve Sprite Scale=DataInterface() {propertyValues {Curve {keys [
        {arriveTangent=7.953, interpMode=RCIM_Cubic, leaveTangent=7.953, tangentMode=RCTM_User, time=-0.0008, value=0.5}
        {interpMode=RCIM_Cubic, leaveTangent=-0.7143, time=0.225, value=1.8998}
        {arriveTangent=-0.1355, interpMode=RCIM_Cubic, leaveTangent=-0.1355, tangentMode=RCTM_User, time=0.6, value=0.1497}
       ]}}})
   /SetVariables(Particles.SpriteSizeFlare2:Vector2f=&MakeVector2D(
     X=&FloatFromCurve(
      FloatCurve=DataInterface() {propertyValues {Curve {keys [{arriveTangent=-2.2776, interpMode=RCIM_Cubic, leaveTangent=-2.2776, tangentMode=RCTM_User, time=0, value=1}, <at 1 0 RCIM_Cubic>]}}}
      Scale Curve=3
     )
     Y=&FloatFromCurve(
      FloatCurve=DataInterface() {propertyValues {Curve {keys [<at 0 1 RCIM_Cubic>, {arriveTangent=-1.3497, interpMode=RCIM_Cubic, leaveTangent=-1.3497, tangentMode=RCTM_User, time=1, value=0.5}]}}}
      Scale Curve=20
     )
    )) {isSetParameters=true, module=SetVariables_#6}
   &Color(Color=&MakeLinearColorFromVectorAndFloat("Vector (RGB)"=<0.4208 0.6361 1>, "Float (Alpha)"=&ClampFloat(Float=0.8))) {enabled=false, module=Color001}
   &DynamicMaterialParameters(
    Index 0 Param 1=&FloatFromCurve(FloatCurve=DataInterface() {propertyValues {Curve {keys [{arriveTangent=2.513, interpMode=RCIM_Cubic, time=0, value=0.2}, <at 1 1.8>]}}})
    Index 0 Param 2=1
    Index 0 Param 3=1
    Index 0 Param 4=1
    Module.Param0WriteEnabled=true
    Module.Param1WriteEnabled=true
    Module.Param2WriteEnabled=true
    Module.Param3WriteEnabled=true
   )
   &CameraOffset(Camera Offset Amount=1)
   /SetVariables(Particles.CameraOffsetFlare=2) {isSetParameters=true, module=SetVariables_#7}
   /SetVariables(Particles.ColorFlare:LinearColor=&ColorFromCurve(ColorCurve=DataInterface() {propertyValues {
       AlphaCurve {keys [<at 0 1>, {arriveTangent=-1, interpMode=RCIM_Cubic, leaveTangent=-1, time=0.5882, value=1}, <at 1 0 RCIM_Cubic>]}
       BlueCurve {keys [<at 0 0.7292 RCIM_Cubic>,<at 0.6045 0.599 RCIM_Cubic>]}
       GreenCurve {keys [<at 0 0.7292 RCIM_Cubic>,<at 0.6045 0.593 RCIM_Cubic>]}
       RedCurve {keys [<at 0 0.7292 RCIM_Cubic>,<at 0.6045 0.4678 RCIM_Cubic>]}
      }})) {isSetParameters=true, module=SetVariables_#8}
   &SolveForcesAndVelocity(Clamp Velocity=false, Limit Acceleration=false)
   &GenerateDeathEvent(Event Probability=1.5, Use Event Probability=true)
  ]
  rendererClasses [&NiagaraMeshRendererProperties, &NiagaraSpriteRendererProperties]
  renderers [
   {rendererClass=&NiagaraMeshRendererProperties, rendererIndex=0}
   {rendererClass=&NiagaraSpriteRendererProperties, rendererIndex=1}
   {rendererClass=&NiagaraSpriteRendererProperties, rendererIndex=2}
  ]
 }
}
summary {
 systemName=NS_Pickaxe_Nucleus_Idle
 emitters [
  {
   emitterName=SheenMesh_start
   rendererClasses [&NiagaraMeshRendererProperties, &NiagaraSpriteRendererProperties]
  }
  {emitterName=Glint_Start, rendererClasses [&NiagaraSpriteRendererProperties]}
  {emitterName=SheenMesh_loop, rendererClasses [&NiagaraMeshRendererProperties, &NiagaraSpriteRendererProperties]}
  {emitterName=Glint_Loop, rendererClasses [&NiagaraSpriteRendererProperties]}
 ]
 userVariables []
}
