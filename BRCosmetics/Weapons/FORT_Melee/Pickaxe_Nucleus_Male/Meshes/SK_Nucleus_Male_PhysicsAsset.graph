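# Physics asset for the male Nucleus pickaxe skeletal mesh, a two-handed melee cosmetic weapon.
# A single capsule body on the Melee_TwoHanded_Handle bone serves as the collision shape.
# It defines no constraints, so the asset has no jointed or secondary bodies.
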PhysicsAsset /BRCosmetics/Weapons/FORT_Melee/Pickaxe_Nucleus_Male/Meshes/SK_Nucleus_Male_PhysicsAsset
bodies {Melee_TwoHanded_Handle [{shapeType=Capsule, length=69.8373, radius=33.8208, center=<-0.0076 -1.1088 27.3879>, rotation=<rot 0.0781 -90 -0.0005>}]}
constraints []
